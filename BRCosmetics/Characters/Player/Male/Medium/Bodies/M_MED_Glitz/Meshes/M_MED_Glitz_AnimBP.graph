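# Post-pose Animation Blueprint for the M_MED_Glitz male player body cosmetic, adding procedural knee and elbow corrections.
# The pose is copied from the attached parent mesh, run through the Glitz Control Rig, then adjusted by additive bone-space offsets on femur helper and cubitus bones.
# Each update derives the four correction alphas from calf-to-thigh and lowerarm-to-upperarm reference-transform deltas, with the right elbow using a narrower input range than the left.
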AnimBlueprint /BRCosmetics/Characters/Player/Male/Medium/Bodies/M_MED_Glitz/Meshes/M_MED_Glitz_AnimBP
target_skeleton=&/M_MED_Glitz_Skeleton
graphs [
 AnimGraph {nodes [
   n1 = Control Rig(Source=Copy Pose From Mesh() {bUseAttachedParent=true}) {
    alphaBoolBlend {blendOption=Linear}
    controlRigAssetReference {blueprintRigClass=&/M_MED_Glitz_CtrlRig.M_MED_Glitz_CtrlRig_C}
   }
   n2 = "Transform (Modify) Bone - Bone: dyn_femur_l_2"(
    ComponentPose=Local To Component(LocalPose=@n1)
    Alpha=Get kneeBend_l() -> (kneeBend_l:double)
    Translation="4.8,2,0"
    Rotation="0,-13,0"
    Scale="1,1,1"
   ) {
    rotationMode=BMM_Additive
    rotationSpace=BCS_BoneSpace
    translationMode=BMM_Additive
    translationSpace=BCS_BoneSpace
    alphaBoolBlend {blendOption=Linear}
    boneToModify {boneName=dyn_femur_l_2}
    rotation {yaw=-13}
    translation=<4.8 2>
   }
   n3 = "Transform (Modify) Bone - Bone: dyn_femur_r_2"(ComponentPose=@n2, Alpha=Get kneeBend_r() -> (kneeBend_r:double), Translation="-4.8,-2,0", Rotation="0,-13,0", Scale="1,1,1") {
    rotationMode=BMM_Additive
    rotationSpace=BCS_BoneSpace
    translationMode=BMM_Additive
    translationSpace=BCS_BoneSpace
    alphaBoolBlend {blendOption=Linear}
    boneToModify {boneName=dyn_femur_r_2}
    rotation {yaw=-13}
    translation=<-4.8 -2>
   }
   n4 = "Transform (Modify) Bone - Bone: cubitus_l"(ComponentPose=@n3, Alpha=Get elbowBend_l() -> (elbowBend_l:double), Translation="4.5,0,0", Rotation="0,0,0", Scale="1,1,1") {
    rotationMode=BMM_Additive
    rotationSpace=BCS_BoneSpace
    translationMode=BMM_Additive
    translationSpace=BCS_BoneSpace
    alphaBoolBlend {blendOption=Linear}
    boneToModify {boneName=cubitus_l}
    translation {x=4.5}
   }
   n5 = "Transform (Modify) Bone - Bone: cubitus_r"(ComponentPose=@n4, Alpha=Get elbowBend_r() -> (elbowBend_r:double), Translation="-4,0,0", Rotation="0,0,0", Scale="1,1,1") {
    rotationMode=BMM_Additive
    rotationSpace=BCS_BoneSpace
    translationMode=BMM_Additive
    translationSpace=BCS_BoneSpace
    alphaBoolBlend {blendOption=Linear}
    boneToModify {boneName=cubitus_r}
    translation {x=-4}
   }
   Output Pose(Result=Component To Local(ComponentPose=@n5)) {layerGroup=None}
  ]}
 EventGraph {nodes [
   Event BlueprintUpdateAnimation()
   > n1 = Sequence!(execute) -> (then_0, then_1, then_2:Exec, then_3:Exec, then_4:Exec)
   > Set bodyMesh(bodyMesh:Skeletal Mesh Component Object Reference=GetOwnerSkeletalMeshComponent()) -> (Output_Get:Skeletal Mesh Component Object Reference)
   Set kneeBend_l(
    execute=@n1.then_1
    kneeBend_l:double=GetAlphaFromReferenceTransformDelta(
     SkeletalMeshComponent=Get bodyMesh() -> (bodyMesh:Skeletal Mesh Component Object Reference)
     BoneName=calf_l
     BaseName=thigh_l
     InputRange=(X=15,Y=100)
     TwistAxis="1,0,0"
    )
   ) -> (Output_Get:double)
   Set kneeBend_r(
    execute=@n1.then_2
    kneeBend_r:double=GetAlphaFromReferenceTransformDelta(SkeletalMeshComponent=Get bodyMesh(), BoneName=calf_r, BaseName=thigh_r, InputRange=(X=15,Y=100), TwistAxis="1,0,0")
   ) -> (Output_Get:double)
   Set elbowBend_l(
    execute=@n1.then_3
    elbowBend_l:double=GetAlphaFromReferenceTransformDelta(SkeletalMeshComponent=Get bodyMesh(), BoneName=lowerarm_l, BaseName=upperarm_l, InputRange=(X=15,Y=90), TwistAxis="1,0,0")
   ) -> (Output_Get:double)
   Set elbowBend_r(
    execute=@n1.then_4
    elbowBend_r:double=GetAlphaFromReferenceTransformDelta(SkeletalMeshComponent=Get bodyMesh(), BoneName=lowerarm_r, BaseName=upperarm_r, InputRange=(X=0,Y=40), TwistAxis="1,0,0")
   ) -> (Output_Get:double)
  ]}
]
variables [
 bodyMesh {type=SkeletalMeshComponent}
 crouchAlpha {type=number, default=1}
 elbowBend_l {type=number}
 elbowBend_r {type=number}
 kneeBend_l {type=number}
 kneeBend_r {type=number}
]
