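# Physics asset for the face accessory mesh of the male medium Glitz player cosmetic, covering six dynamic hair bones and the head.
# Three left and three right hair bones each carry an identical near-spherical capsule, offset along X in opposite directions per side.
# Each hair body is constrained directly to a small head capsule rather than chained, with tight swing and twist limits that keep movement subtle.
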PhysicsAsset /BRCosmetics/Characters/Player/Male/Medium/Bodies/M_MED_Glitz/Meshes/Parts/M_MED_Glitz_FaceAcc_Physics
bodies {
 dyn_hair_l_1 [{shapeType=Capsule, length=0.3048, radius=4.6231, shapeName=dyn_hair_l_1_capsule, center=<6.3983 -0.0006 0>}]
 dyn_hair_l_2 [{shapeType=Capsule, length=0.3048, radius=4.6231, shapeName=dyn_hair_l_2_capsule, center=<6.3983 -0.0006 0>}]
 dyn_hair_l_3 [{shapeType=Capsule, length=0.3048, radius=4.6231, shapeName=dyn_hair_l_3_capsule, center=<6.3983 -0.0006 0>}]
 dyn_hair_r_1 [{shapeType=Capsule, length=0.3048, radius=4.6231, shapeName=dyn_hair_r_1_capsule, center=<-4.8664 -0.0004 0>}]
 dyn_hair_r_2 [{shapeType=Capsule, length=0.3048, radius=4.6231, shapeName=dyn_hair_r_2_capsule, center=<-4.8664 -0.0004 0>}]
 dyn_hair_r_3 [{shapeType=Capsule, length=0.3048, radius=4.6231, shapeName=dyn_hair_r_3_capsule, center=<-4.8664 -0.0004 0>}]
 head [{shapeType=Capsule, length=1, radius=1, shapeName=head_capsule}]
}
constraints [
 {bone1Name=dyn_hair_l_1, bone2Name=head, swing1LimitDegrees=10, swing2LimitDegrees=5, twistLimitDegrees=5}
 {bone1Name=dyn_hair_l_2, bone2Name=head, swing1LimitDegrees=10, swing2LimitDegrees=5, twistLimitDegrees=5}
 {bone1Name=dyn_hair_l_3, bone2Name=head, swing1LimitDegrees=10, swing2LimitDegrees=5, twistLimitDegrees=5}
 {bone1Name=dyn_hair_r_1, bone2Name=head, swing1LimitDegrees=10, swing2LimitDegrees=5, twistLimitDegrees=5}
 {bone1Name=dyn_hair_r_2, bone2Name=head, swing1LimitDegrees=10, swing2LimitDegrees=5, twistLimitDegrees=5}
 {bone1Name=dyn_hair_r_3, bone2Name=head, swing1LimitDegrees=10, swing2LimitDegrees=5, twistLimitDegrees=5}
]
