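# Animation Blueprint for the face accessory part of the Glitz medium male player body, adding physics motion and a targeting head tilt.
# It copies the attached parent mesh's pose, then runs RigidBody simulation in base-bone space with overridden gravity.
# An additive head pitch offset follows, weighted by Is Targeting.
# Each update sets that weight from bIsTargeting through FInterpTo, whose zero DeltaTime makes it switch without easing.
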AnimBlueprint /BRCosmetics/Characters/Player/Male/Medium/Bodies/M_MED_Glitz/Meshes/Parts/M_MED_Glitz_FaceAcc_AnimBP
target_skeleton=&/M_MED_Glitz_FaceAcc_Skeleton
graphs [
 AnimGraph {nodes [
   n1 = RigidBody(ComponentPose=Local To Component(LocalPose=Copy Pose From Mesh() {bUseAttachedParent=true})) {
    bClampLinearTranslationLimitToRefPose=true
    bOverrideWorldGravity=true
    overridePhysicsAsset=&/M_MED_Glitz_FaceAcc_Physics
    simulationSpace=BaseBoneSpace
    alphaBoolBlend {blendOption=Linear}
    componentAppliedLinearAccClamp=<3000 3000 3000>
    componentLinearAccScale=<0.1 0.05 0.05>
    componentLinearVelScale=<0.1 0.05 0.05>
    overrideWorldGravity {z=-980}
   }
   n2 = "Transform (Modify) Bone - Bone: head"(ComponentPose=@n1, Alpha=Get Is Targeting() -> (Is Targeting:double), Translation="0,0,0", Rotation="-10,0,0", Scale="1,1,1") {rotationMode=BMM_Additive, scaleMode=BMM_Additive, alphaBoolBlend {blendOption=Linear}, boneToModify {boneName=head}, rotation {pitch=-10}}
   Output Pose(Result=Component To Local(ComponentPose=@n2)) {layerGroup=None}
  ]}
 EventGraph {nodes [
   Event BlueprintUpdateAnimation()
   n1 = Get bIsTargeting()
   > Set Is Targeting(Is Targeting:double=FInterpTo(Current="To Float (Boolean)"(InBool=@n1), Target=1, DeltaTime=0, InterpSpeed=0.5)) -> (Output_Get:double)
  ]}
]
variables [Is Targeting {type=number}]
